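# Parent fur material for the ChubbyJingle large male player body, layering shell-mesh fur onto the shared 2023 character parent functions.
# Character PreFX attributes and masks feed a chain of fur functions for custom color, adjustments, painting, shells and WPO shrink.
# The result gets masked root color, subsurface color tweaks, shell-layer-driven subsurface opacity and faint base-color emissive.
# A PostFX character function is also wired in through named reroutes.
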Material /BRCosmetics/Characters/Player/Male/Large/Bodies/M_LRG_ChubbyJingle/Materials/Fur/M_ChubbyJingle_FurParent
graph {
 properties {
  bUseMaterialAttributes=true
  bUsedWithClothing=true
  bUsedWithInstancedStaticMeshes=true
  bUsedWithMeshDeformer=true
  bUsedWithMorphTargets=true
  bUsedWithNanite=true
  bUsedWithSkeletalMesh=true
  pixelDepthOffsetMode=PDOM_Legacy
  shadingModel=MSM_FromMaterialExpression
  usedShadingModels=Default Lit | Subsurface | Clear Coat | Subsurface Profile | Cloth
 }
 nodes [
  Constant()
  n1 = MaterialFunctionCall() {materialFunction=&MF_CharacterParent_2023_PreFX}
  MaterialFunctionCall(Material=NamedRerouteUsage(), M=@n1.M, Pre Skinned Local Position=@n1.M, Distance Blend=@n1.M, Fresnel=@n1.M) {materialFunction=&MF_CharacterParent_2023_PostFX}
  NamedRerouteDeclaration(None=@n1.Tech Art Mask) {name=TechArtMask, nodeColor=<rgba 0.2188 0 1 1>}
  NamedRerouteDeclaration(None=@n1.Result) {name=Material, nodeColor=<rgba 0 1 0.1094 1>}
  NamedRerouteDeclaration(None=@n1.M) {name=Mask, nodeColor=<rgba 1 0 0.2109 1>}
  n2 = MaterialFunctionCall!(Material=NamedRerouteUsage(), TechArtMask=NamedRerouteUsage()) {materialFunction=/Game/Global/Materials/Character_Parent_2023/Fur_Parent_2023/MF_Fur_CustomColor}
  n3 = NamedRerouteUsage()
  n4 = MaterialFunctionCall!(
   In=MaterialFunctionCall!(
    In=MaterialFunctionCall!(MA=@n2.Result, TechArtMask=@n3) {materialFunction=/Game/Global/Materials/Character_Parent_2023/Fur_Parent_2023/MF_Fur_ParamAdjustments}.Result
    TechArtMask=NamedRerouteUsage()
    Mask Input=NamedRerouteUsage()
   ) {materialFunction=/Game/Global/Materials/Character_Parent_2023/Fur_Parent_2023/MF_Fur_Painted}.Result
   TechArtMask=NamedRerouteUsage()
  ) {materialFunction=/Game/Global/Materials/Character_Parent_2023/Fur_Parent_2023/MF_Fur_Shells}
  n5 = GetMaterialAttributes(None=MaterialFunctionCall!(Material=@n4.Result, Mask=NamedRerouteUsage(), Fur Length=@n4.Result) {materialFunction=/Game/Global/Materials/Character_Parent_2023/Fur_Parent_2023/MF_FurWPOShrink}.Result) {attributeGetTypes [#1, #2, #3]}
  n6 = ScalarParameter() {defaultValue=1, group=FurbaseColor, parameterName=FurbaseRootMulitplier}
  n7 = StaticComponentMaskParameter(None=NamedRerouteUsage()) {defaultG=true, group=FurbaseColor, parameterName=FurbaseRootMulitplier_Channel}
  n8 = NamedRerouteUsage()
  n9 = ShellMeshMaterialExpressionNormalizedShellLayer() {is=class ShellMeshMaterialExpressionNormalizedShellLayer}
  n10 = VertexInterpolator(VS=@n9)
  n11 = LinearInterpolate(
   B=ScalarParameter() {defaultValue=0.9, group=FurbaseColor, parameterName=FurShellSubsurfaceOpacity}
   Alpha=NamedRerouteDeclaration(None=ShaderStageSwitch(PixelShader=@n10.PS, VertexShader=@n9)) {name=NormalizedShellLayer, nodeColor=<rgba 0.2656 0 1 1>}
  ) {constA=1}
  n12 = SetMaterialAttributes!(
   MaterialAttributes=@n5.MaterialAttributes
   Base Color=Multiply(A=NamedRerouteUsage(), B=LinearInterpolate(B=@n6, Alpha=@n7) {constA=1})
   Fuzz Color or Subsurface Color=LinearInterpolate(A=NamedRerouteUsage(), B=Multiply(A=@n8) {constB=0.05}, Alpha=NamedRerouteUsage())
   Opacity=LinearInterpolate(
    A=NamedRerouteUsage()
    B=NamedRerouteDeclaration(None=@n11) {name=FurShellSubsurfaceOpacity, nodeColor=<rgba 0 1 0.25 1>}
    Alpha=NamedRerouteUsage()
   )
   Emissive Color=Multiply(A=@n5.Base Color) {constB=0.02}
  ) {attributeSetTypes [
    #1
    #3
    #2
    #4
   ]}
  NamedRerouteDeclaration(None=@n12) {name=OUT_Material, nodeColor=<rgba 0 0.5312 1 1>}
  NamedRerouteDeclaration(None=@n5.Opacity) {name=InputOpacity, nodeColor=<rgba 1 0 0.0703 1>}
  NamedRerouteDeclaration(None=@n5.Fuzz Color or Subsurface Color) {name=InputSubsurfaceColor, nodeColor=<rgba 0 0.3906 1 1>}
  NamedRerouteDeclaration(None=@n5.Base Color) {name=InputBaseColor, nodeColor=<rgba 0 0.4609 1 1>}
 ]
}
